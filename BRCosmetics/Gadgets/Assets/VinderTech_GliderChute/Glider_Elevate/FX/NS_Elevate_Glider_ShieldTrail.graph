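# Niagara system for a short-lived shield trail ribbon on the Elevate glider chute, named for VinderTech.
# Its single attached-ribbon emitter spawns and updates points from the user skeletal mesh, offset along Z, and inherits source movement.
# Alpha fades over each particle's life, multiplied by a half-strength fade-in and fade-out curve driven by normalized system loop age.
# The integrator and upward-velocity modules are disabled.
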NiagaraSystem /BRCosmetics/Gadgets/Assets/VinderTech_GliderChute/Glider_Elevate/FX/NS_Elevate_Glider_ShieldTrail
dependencies {usedDataInterfaces [&NiagaraDataInterfaceCurve], usedDynamicInputs [], usedModules [&SystemState]}
emitters {FN_AttachedRibbon {
  emitterUpdateScript [
   &EmitterState(Life Cycle Mode=::NewEnumerator1, Loop Duration=0.5, UseLoopDelay=false)
   &FN_SpawnRibbon_v2(Spawn Mode=::NewEnumerator0, SpawnRate=30)
   &FN_Integrator() {enabled=false}
  ]
  particleSpawnScript [
   /Game/Effects/Niagara/Modules/Ribbons/FN_InitializeRibbon_v4(
    Enable Skeletal Position=true
    Lifetime=0.15
    Offset Axis:ENiagaraOrientationAxis=::ZAxis
    Ribbon Width=53
    SkeletalMesh:NiagaraDataInterfaceSkeletalMesh=$User.SkeletalMesh
    Write Material Random=false
   )
   &FN_FlipRibbonV(Offset=1)
   &AddVelocity(Coordinate Space=::World, Velocity=<0 0 50>) {enabled=false}
  ]
  particleUpdateScript [
   &InheritSourceMovement(ApplyForce=false, ApplyPosition=true, ApplyPositionFalloff=false)
   &SolveForcesAndVelocity(Clamp Velocity=false, Limit Acceleration=false)
   /Game/Effects/Niagara/Modules/Ribbons/FN_UpdateRibbon_v4(Enable Skeletal Position=true, SkeletalMesh:NiagaraDataInterfaceSkeletalMesh=$User.SkeletalMesh)
   &FN_RibbonMaterial()
   &ScaleColor(
    Scale Alpha=&FloatFromCurve(
     FloatCurve=DataInterface() {propertyValues {Curve {keys [<at 0 1>,<at 1 0>]}}}
     Scale Curve=&FloatFromCurve(
      FloatCurve=DataInterface() {propertyValues {Curve {keys [
          {interpMode=RCIM_Cubic, tangentMode=RCTM_User, time=0, value=0}
          <at 0.15 1>
          {interpMode=RCIM_Cubic, tangentMode=RCTM_User, time=0.85, value=1}
          {tangentMode=RCTM_User, time=1, value=0}
         ]}}}
      CurveIndex=$System.NormalizedLoopAge
      Scale Curve=0.5
     )
    )
    ScaleA=true
    ScaleRGB=true
   )
  ]
  rendererClasses [&NiagaraRibbonRendererProperties]
  renderers [{rendererClass=&NiagaraRibbonRendererProperties}]
 }}
summary {
 systemName=NS_Elevate_Glider_ShieldTrail
 emitters [{emitterName=FN_AttachedRibbon, rendererClasses [&NiagaraRibbonRendererProperties]}]
 userVariables [
  User.RibbonIDOffset {type=NiagaraInt32}
  User.SkeletalMesh {
   type=NiagaraDataInterfaceSkeletalMesh
   defaultValue {
    struct=NiagaraExt_VariableValue_DataInterface
    value {dataInterface=&self:NiagaraDataInterfaceSkeletalMesh_1}
   }
  }
 ]
}
